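# Niagara fire effect for the Cyclone emote that follows the character's right-hand weapon attach points.
# A single fading mesh particle tracks the weapon_r bone, while glow sprites, animated SubUV flames and rising sparks spawn continuously around the Melee_R_Lower socket.
# The sprite emitters resample the skeleton in particle update, adding random spawn-group offsets rotated by the sampled rotation.
# Curves fade and resize the sprites, and velocity drifts flames and sparks upward.
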NiagaraSystem /BRCosmetics/Athena/Effects/Particles/Emotes/NS_Emote_Cyclone/NS_Emote_Cyclone_01
dependencies {
 usedDataInterfaces [&NiagaraDataInterfaceCurve]
 usedDynamicInputs []
 usedModules [/Niagara/Modules/System/SystemLifeCycle, &SystemState]
}
emitters {
 FN_Emote_FireDecal {
  emitterUpdateScript [
   &EmitterScalability(
    Module.bApplySpawnCountScale=false
    Module.bApplySpawnCountScaleByDistanceFraction=false
    Module.bUseMaxDistance=false
    Module.bUseMinDistance=false
   ) {enabled=false}
   &EmitterLifeCycle(MaxLoopCount=1, NextLoopDuration=0.1) {enabled=false}
   &EmitterState(Life Cycle Mode=::NewEnumerator1, Loop Duration=1, UseLoopDelay=false)
   &SpawnBurst_Instantaneous(Use Spawn Probability=false) {enabled=false}
   &SpawnRate(SpawnRate=30, Use Spawn Probability=false)
   &SpawnBurst_Instantaneous(Spawn Count=2, Use Spawn Probability=false) {enabled=false, module=SpawnBurst_Instantaneous001}
   /Game/Effects/Niagara/Modules/Spawn/SpawnRateFromMultipleSpawnGroups_2(Global Spawnrate=2) {enabled=false}
  ]
  particleSpawnScript [
   /Niagara/Modules/Spawn/Initialization/InitializeParticle(
    Lifetime=&UniformRangedFloat(Minimum=1.5, Maximum=1.2)
    Module.Write Color=true
    Module.Write Lifetime=true
    Module.Write Mass=true
    Module.Write Material Random=true
    Module.Write Position=true
    Module.Write Scale=false
    Module.Write SpriteRotation=true
    Module.Write SpriteSize=true
    Position=&SimulationPosition()
    Sprite Rotation=&UniformRangedFloat(Minimum=0, Maximum=0)
    Sprite Size=&Vector2DFromFloat(Value=&UniformRangedFloat(Minimum=15, Maximum=20))
   )
   /SetVariables(Particles.Distribution_L:Vector3f=&UniformRangedVector(Minimum=<10 0 -50>, Maximum=<5 0 70>)) {isSetParameters=true, module=SetVariables_#1}
   /SetVariables(Particles.Distribution_R:Vector3f=&UniformRangedVector(Minimum=<10 0 -50>, Maximum=<5 0 70>)) {isSetParameters=true, module=SetVariables_#2}
   &CameraOffset(Camera Offset Amount=5)
   /Game/Effects/Niagara/Modules/Groups/FN_SetSpawnGroup_v1/FN_SetSpawnGroup(Emitter Bone Count=2)
   &DynamicMaterialParameters(
    Index 0 Param 1=1.7
    Index 0 Param 2=1
    Index 0 Param 3=1
    Index 0 Param 4=&UniformRangedFloat(Minimum=0, Maximum=1)
    Module.Param0WriteEnabled=true
    Module.Param1WriteEnabled=true
    Module.Param2WriteEnabled=true
    Module.Param3WriteEnabled=true
   )
  ]
  particleUpdateScript [
   &ParticleState()
   &SolveForcesAndVelocity(Clamp Velocity=false, Limit Acceleration=false)
   &ScaleColor(
    Scale Alpha=&FloatFromCurve(
     FloatCurve=DataInterface() {propertyValues {Curve {keys [
         {interpMode=RCIM_Cubic, leaveTangent=0.6032, tangentMode=RCTM_Break, time=0, value=0}
         {arriveTangent=-0.0426, interpMode=RCIM_Cubic, leaveTangent=-0.0426, tangentMode=RCTM_User, time=0.3, value=1}
         {arriveTangent=0.0145, interpMode=RCIM_Cubic, leaveTangent=0.0145, tangentMode=RCTM_User, time=0.6, value=1}
         <at 1 0>
        ]}}}
     Scale Curve=0.5
    )
    Scale RGB=<100 20 0>
    ScaleA=true
    ScaleRGB=true
   )
   &ScaleSpriteSize(Scale Factor=&Vector2DFromFloat(Value=&FloatFromCurve(FloatCurve=DataInterface() {propertyValues {Curve {keys [<at 0.0121 0.5565>,<at 0.491 0.8934>,<at 1.0011 0.002>]}}})))
   &SampleSkeletalMeshSkeleton(
    Mesh=DataInterface() {propertyValues {
      DefaultMesh=null
      ExcludeBoneName=root
      OptionalComponentClass=null
      PreviewMesh=/Game/Characters/Player/Male/Male_Avg_Base/Fortnite_M_Avg_Player
      SkinningMode=SkinOnTheFly
      SoftSourceActor=null
      SourceMode=Default
      UvSetIndex=0
      WholeMeshLOD=-1
      bExcludeBone=false
      bReadDeformedGeometry=true
      bRequireCurrentFrameData=true
      ComponentTags []
      FilteredBones []
      FilteredSockets [Melee_R_Lower]
      MeshUserParameter {parameter {name=None, typeDefHandle {registeredTypeIndex=1}, varData []}}
      SamplingRegions []
     }}
    Spawn Mode:ENiagaraSkeletalSpawning=::NewEnumerator2
    Specific Bone Index:NiagaraInt32=$Particles.SpawnGroup
   )
   /Niagara/Modules/Spawn/Location/SkeletalMeshSkeletonLocation(
    Module.Inherit Velocity=false
    Module.Use Sampled Bone as Ribbon ID=true
    Sampled Position Offset:Vector3f=&RotateVectorByQuaternion(
     Vector To Rotate=/Game/Effects/Niagara/DynamicInputs/Water/FN_PickVectorFromSpawnGroupCutoff(
      Under Cutoff Vector:Vector3f=$Particles.Distribution_L
      Over Cutoff Vector:Vector3f=$Particles.Distribution_R
      SpawnGroup:NiagaraInt32=$Particles.SpawnGroup
      Cutoff:NiagaraInt32=1
     )
     Quaternion=$Particles.SampleSkeletalMeshSkeleton.SampledRotation
    )
   )
  ]
  rendererClasses [&NiagaraSpriteRendererProperties]
  renderers [{rendererClass=&NiagaraSpriteRendererProperties}]
 }
 FN_Emote_FireMesh {
  emitterUpdateScript [
   &EmitterState(Inactive Response=::NewEnumerator1, Life Cycle Mode=::NewEnumerator1, Loop Duration=1, UseLoopDelay=false)
   &SpawnBurst_Instantaneous(Spawn Count=1, Spawn Time=0.1, Use Spawn Probability=false)
  ]
  particleSpawnScript [&InitializeParticle(
    Lifetime=0.999
    Module.Write Color=true
    Module.Write Lifetime=true
    Module.Write Mass=true
    Module.Write Material Random=true
    Module.Write Position=false
    Module.Write Scale=false
    Module.Write SpriteRotation=false
    Module.Write SpriteSize=true
    Position=&SimulationPosition()
    Sprite Size=<400 400>
   )]
  particleUpdateScript [
   &ParticleState()
   &SampleSkeletalMeshSkeleton(Mesh=DataInterface() {propertyValues {
      DefaultMesh=null
      ExcludeBoneName=root
      OptionalComponentClass=null
      PreviewMesh=&Fortnite_M_Avg_Player
      SkinningMode=SkinOnTheFly
      SoftSourceActor=null
      SourceMode=Default
      UvSetIndex=0
      WholeMeshLOD=-1
      bExcludeBone=false
      bReadDeformedGeometry=true
      bRequireCurrentFrameData=true
      ComponentTags []
      FilteredBones [weapon_r]
      FilteredSockets []
      MeshUserParameter {parameter {name=None, typeDefHandle {registeredTypeIndex=1}, varData []}}
      SamplingRegions []
     }})
   &SkeletalMeshSkeletonLocation(
    Inherit Velocity From Mesh=0
    Module.Inherit Velocity=true
    Module.Use Sampled Bone as Ribbon ID=true
    Sampled Position Offset:Vector3f=<0 0 -3>
   )
   &ScaleColor(
    Scale Alpha=&FloatFromCurve(FloatCurve=DataInterface() {propertyValues {Curve {keys [
         {arriveTangent=-0.551, interpMode=RCIM_Cubic, leaveTangent=-0.551, tangentMode=RCTM_User, time=0, value=1}
         {arriveTangent=-2.3347, interpMode=RCIM_Cubic, leaveTangent=-2.3347, tangentMode=RCTM_User, time=1.0011, value=0.002}
        ]}}})
    ScaleA=true
    ScaleRGB=true
   )
  ]
  rendererClasses [&NiagaraMeshRendererProperties]
  renderers [{rendererClass=&NiagaraMeshRendererProperties, rendererIndex=0}]
 }
 FN_Emote_FireParticleSubuv {
  emitterUpdateScript [
   &EmitterScalability(
    Module.bApplySpawnCountScale=false
    Module.bApplySpawnCountScaleByDistanceFraction=false
    Module.bUseMaxDistance=false
    Module.bUseMinDistance=false
   ) {enabled=false}
   &EmitterLifeCycle(MaxLoopCount=1, NextLoopDuration=0.1) {enabled=false}
   &EmitterState(Life Cycle Mode=::NewEnumerator1, Loop Duration=1, UseLoopDelay=false)
   &SpawnBurst_Instantaneous(Use Spawn Probability=false) {enabled=false}
   &SpawnRate(SpawnRate=8, Use Spawn Probability=false)
   &SpawnBurst_Instantaneous(Spawn Count=2, Use Spawn Probability=false) {enabled=false, module=SpawnBurst_Instantaneous001}
   &SpawnRateFromMultipleSpawnGroups_2(Global Spawnrate=2) {enabled=false}
  ]
  particleSpawnScript [
   &InitializeParticle(
    Color=<rgba 0.1302 0.0315 0.0118 1>
    Lifetime=&UniformRangedFloat(Minimum=0.6, Maximum=0.9)
    Module.Write Color=true
    Module.Write Lifetime=true
    Module.Write Mass=true
    Module.Write Material Random=true
    Module.Write Position=true
    Module.Write Scale=false
    Module.Write SpriteRotation=true
    Module.Write SpriteSize=true
    Position=&SimulationPosition()
    Sprite Rotation=&UniformRangedFloat(Minimum=0, Maximum=360)
    Sprite Size=&UniformRangedVector2D(Minimum=<8 8>, Maximum=<12 12>)
   )
   &SystemLocation(MaskSpawnGroup=false, Offset=<0 -5 180>)
   /SetVariables(Particles.Distribution_L:Vector3f=&UniformRangedVector(Minimum=<5 0 -40>, Maximum=<0 10 60>)) {isSetParameters=true, module=SetVariables_#1}
   /SetVariables(Particles.Distribution_R:Vector3f=&UniformRangedVector(Minimum=<5 0 -40>, Maximum=<0 10 60>)) {isSetParameters=true, module=SetVariables_#2}
   &CameraOffset(Camera Offset Amount=5)
   &FN_SetSpawnGroup(Emitter Bone Count=2)
  ]
  particleUpdateScript [
   &ParticleState()
   &SampleSkeletalMeshSkeleton(
    Mesh=DataInterface() {propertyValues {
      DefaultMesh=null
      ExcludeBoneName=root
      OptionalComponentClass=null
      PreviewMesh=&Fortnite_M_Avg_Player
      SkinningMode=SkinOnTheFly
      SoftSourceActor=null
      SourceMode=Default
      UvSetIndex=0
      WholeMeshLOD=-1
      bExcludeBone=false
      bReadDeformedGeometry=true
      bRequireCurrentFrameData=true
      ComponentTags []
      FilteredBones []
      FilteredSockets [Melee_R_Lower]
      MeshUserParameter {parameter {name=None, typeDefHandle {registeredTypeIndex=1}, varData []}}
      SamplingRegions []
     }}
    Spawn Mode:ENiagaraSkeletalSpawning=::NewEnumerator2
    Specific Bone Index:NiagaraInt32=$Particles.SpawnGroup
   )
   &SkeletalMeshSkeletonLocation(
    Inherit Velocity From Mesh=0.2
    Module.Inherit Velocity=true
    Module.Use Sampled Bone as Ribbon ID=true
    Sampled Position Attribute:NiagaraPosition=$Particles.Initial.Position
    Sampled Position Offset:Vector3f=&RotateVectorByQuaternion(
     Vector To Rotate=&FN_PickVectorFromSpawnGroupCutoff(
      Under Cutoff Vector:Vector3f=$Particles.Distribution_L
      Over Cutoff Vector:Vector3f=$Particles.Distribution_R
      SpawnGroup:NiagaraInt32=$Particles.SpawnGroup
      Cutoff:NiagaraInt32=1
     )
     Quaternion=$Particles.SampleSkeletalMeshSkeleton.SampledRotation
    )
   )
   &AddVelocity(Velocity=&UniformRangedVector(Minimum=<0 0 90>, Maximum=<0 0 20>))
   &SolveForcesAndVelocity(Clamp Velocity=false, Limit Acceleration=false)
   &ScaleColor(
    Scale Alpha=&FloatFromCurve(
     FloatCurve=DataInterface() {propertyValues {Curve {keys [
         {interpMode=RCIM_Cubic, leaveTangent=0.6032, tangentMode=RCTM_Break, time=0, value=0}
         {arriveTangent=-0.0426, interpMode=RCIM_Cubic, leaveTangent=-0.0426, tangentMode=RCTM_User, time=0.3, value=1}
         {arriveTangent=0.0145, interpMode=RCIM_Cubic, leaveTangent=0.0145, tangentMode=RCTM_User, time=0.6, value=1}
         <at 1 0>
        ]}}}
     Scale Curve=0.5
    )
    ScaleA=true
    ScaleRGB=true
   )
   &ScaleSpriteSize(Scale Factor=&Vector2DFromFloat(Value=&FloatFromCurve(FloatCurve=DataInterface() {propertyValues {Curve {keys [<at 0.0034 0.1274>,<at 0.2976 0.9854>,<at 1.0011 0.002>]}}})))
   /SetVariables(Particles.SubImageIndex=&FloatFromCurve(FloatCurve=DataInterface() {propertyValues {Curve {keys [<at 0 0>,<at 1 15>]}}})) {isSetParameters=true}
  ]
  rendererClasses [&NiagaraSpriteRendererProperties]
  renderers [{rendererClass=&NiagaraSpriteRendererProperties}]
 }
 FN_Emote_Spark {
  emitterUpdateScript [
   &EmitterScalability(
    Module.bApplySpawnCountScale=false
    Module.bApplySpawnCountScaleByDistanceFraction=false
    Module.bUseMaxDistance=false
    Module.bUseMinDistance=false
   ) {enabled=false}
   &EmitterLifeCycle(MaxLoopCount=1, NextLoopDuration=0.1) {enabled=false}
   &EmitterState(Life Cycle Mode=::NewEnumerator1, Loop Duration=1, UseLoopDelay=false)
   &SpawnBurst_Instantaneous(Use Spawn Probability=false) {enabled=false}
   &SpawnRate(SpawnRate=3, Use Spawn Probability=false)
   &SpawnBurst_Instantaneous(Spawn Count=2, Use Spawn Probability=false) {enabled=false, module=SpawnBurst_Instantaneous001}
   &SpawnRateFromMultipleSpawnGroups_2(Global Spawnrate=2) {enabled=false}
  ]
  particleSpawnScript [
   /SetVariables(Particles.Velocity=&UniformRangedVector(Minimum=<500 0 0>, Maximum=<-500 0 0>)) {enabled=false, isSetParameters=true}
   /Game/Effects/Niagara/Modules/Orbit/FN_Orbit_Init_Variables_v2()
   &InitializeParticle(
    Color=<rgba 10 1.5 0.5 1>
    Lifetime=&UniformRangedFloat(Minimum=0.6, Maximum=1.1)
    Module.Write Color=true
    Module.Write Lifetime=true
    Module.Write Mass=true
    Module.Write Material Random=true
    Module.Write Position=true
    Module.Write Scale=false
    Module.Write SpriteRotation=true
    Module.Write SpriteSize=true
    Position=&SimulationPosition()
    Sprite Rotation=&UniformRangedFloat(Minimum=360, Maximum=0)
    Sprite Size=&UniformRangedVector2D(Minimum=<0.7 3.5>, Maximum=<1.5 2>)
   )
   &SystemLocation(MaskSpawnGroup=false, Offset=<0 -5 190>)
   /SetVariables(Particles.Distribution_L:Vector3f=&UniformRangedVector(Minimum=<5 0 -40>, Maximum=<0 10 60>)) {isSetParameters=true, module=SetVariables_#1}
   /SetVariables(Particles.Distribution_R:Vector3f=&UniformRangedVector(Minimum=<5 0 -40>, Maximum=<0 10 60>)) {isSetParameters=true, module=SetVariables_#2}
   &CameraOffset(Camera Offset Amount=5)
   &FN_SetSpawnGroup(Emitter Bone Count=2)
  ]
  particleUpdateScript [
   &ParticleState()
   &SampleSkeletalMeshSkeleton(
    Mesh=DataInterface() {propertyValues {
      DefaultMesh=null
      ExcludeBoneName=root
      OptionalComponentClass=null
      PreviewMesh=&Fortnite_M_Avg_Player
      SkinningMode=SkinOnTheFly
      SoftSourceActor=null
      SourceMode=Default
      UvSetIndex=0
      WholeMeshLOD=-1
      bExcludeBone=false
      bReadDeformedGeometry=true
      bRequireCurrentFrameData=true
      ComponentTags []
      FilteredBones []
      FilteredSockets [Melee_R_Lower]
      MeshUserParameter {parameter {name=None, typeDefHandle {registeredTypeIndex=1}, varData []}}
      SamplingRegions []
     }}
    Spawn Mode:ENiagaraSkeletalSpawning=::NewEnumerator2
    Specific Bone Index:NiagaraInt32=$Particles.SpawnGroup
   )
   &SkeletalMeshSkeletonLocation(
    Inherit Velocity From Mesh=0
    Module.Inherit Velocity=false
    Module.Use Sampled Bone as Ribbon ID=true
    Sampled Position Attribute:NiagaraPosition=$Particles.Initial.Position
    Sampled Position Offset:Vector3f=&RotateVectorByQuaternion(
     Vector To Rotate=&FN_PickVectorFromSpawnGroupCutoff(
      Under Cutoff Vector:Vector3f=$Particles.Distribution_L
      Over Cutoff Vector:Vector3f=$Particles.Distribution_R
      SpawnGroup:NiagaraInt32=$Particles.SpawnGroup
      Cutoff:NiagaraInt32=1
     )
     Quaternion=$Particles.SampleSkeletalMeshSkeleton.SampledRotation
    )
   )
   &AddVelocity(
    Coordinate Space=::Simulation
    Velocity=&UniformRangedVector(Minimum=<30 0 50>, Maximum=<-30 0 30>, Randomness Mode=::NewEnumerator1, Random Seed=5, Fixed Random Seed=true)
   )
   &AddVelocity(Coordinate Space=::World, Velocity=<30 0 0>) {module=AddVelocity001}
   /Game/Effects/Niagara/Modules/Orbit/FN_Orbit_v3(
    Height min/max:Vector2f=<0 50>
    Radius min/max:Vector2f=<0 1>
    Rotation Offset min/max:Vector2f=<0 1>
    Rotation Rate min/max:Vector2f=<0 1>
    To Axis:Vector3f=<0 0 0>
   ) {enabled=false}
   &Drag(Drag=1.5, Use Linear Drag=true, Use Rotational Drag=false) {enabled=false}
   &SolveForcesAndVelocity(Clamp Velocity=false, Limit Acceleration=false)
   &ScaleColor(
    Scale Alpha=&FloatFromCurve(
     FloatCurve=DataInterface() {propertyValues {Curve {keys [
         {interpMode=RCIM_Cubic, leaveTangent=0.6032, tangentMode=RCTM_Break, time=0, value=0}
         {arriveTangent=-0.0426, interpMode=RCIM_Cubic, leaveTangent=-0.0426, tangentMode=RCTM_User, time=0.3, value=1}
         {arriveTangent=0.0145, interpMode=RCIM_Cubic, leaveTangent=0.0145, tangentMode=RCTM_User, time=0.6, value=1}
         <at 1 0>
        ]}}}
     Scale Curve=0.5
    )
    Scale RGB=&VectorFromFloat(Value=5)
    ScaleA=true
    ScaleRGB=true
   )
   &ScaleSpriteSize(Scale Factor=&Vector2DFromFloat(Value=&FloatFromCurve(FloatCurve=DataInterface() {propertyValues {Curve {keys [<at 0.0034 0.1274>,<at 0.2976 0.9854>,<at 1.0011 0.002>]}}})))
  ]
  rendererClasses [&NiagaraSpriteRendererProperties]
  renderers [{rendererClass=&NiagaraSpriteRendererProperties}]
 }
}
summary {
 systemName=NS_Emote_Cyclone_01
 emitters [
  {emitterName=FN_Emote_FireMesh, rendererClasses [&NiagaraMeshRendererProperties]}
  {emitterName=FN_Emote_FireDecal, rendererClasses [&NiagaraSpriteRendererProperties]}
  {emitterName=FN_Emote_FireParticleSubuv, rendererClasses [&NiagaraSpriteRendererProperties]}
  {emitterName=FN_Emote_Spark, rendererClasses [&NiagaraSpriteRendererProperties]}
 ]
 userVariables [
  User.bIsFrontEndPreview {type=NiagaraBool}
  User.PlayerSpeed {type=NiagaraFloat, defaultValue=0, description=""}
  User.RibbonIDOffset {type=NiagaraInt32}
  User.SkeletalMesh {
   type=NiagaraDataInterfaceSkeletalMesh
   defaultValue {
    struct=NiagaraExt_VariableValue_DataInterface
    value {dataInterface=&self:NiagaraDataInterfaceSkeletalMesh_0}
   }
  }
 ]
}
